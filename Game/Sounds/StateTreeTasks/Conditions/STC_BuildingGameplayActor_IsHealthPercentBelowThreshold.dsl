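; State Tree condition that tests whether a building gameplay actor's health percent is below a threshold.
; It reads the Building Gameplay Actor input's health percent and compares it to the Threshold input with a strict less-than.
; The Invert parameter flips the result to greater-than-or-equal and also switches the node's description text to match.
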(fn GetDescription
  (params (Formatting :type EStateTreeNodeFormatting%20Enum :default "RichText"))
  (return (select (Variables|Parameter|GetInvert) "NSLOCTEXT(\"[2C1B1C1455D792CB942856CF81C95FE4]\", \"D191FD5B48414807BE158F9D8561A919\", \"Is Building Health Percent Greater Than or Equal To Threshold\")" "NSLOCTEXT(\"[2C1B1C1455D792CB942856CF81C95FE4]\", \"B39771FC439F41631BD6AEA3FC456729\", \"Is Building Health Percent Less Than Threshold\")")))

(fn ReceiveTestCondition
  (return (!= (< (Health|GetHealthPercent (Variables|Input|GetBuilding%20Gameplay%20Actor) :!class /Script/FortniteGame.BuildingActor) (Variables|Input|GetThreshold)) (Variables|Parameter|GetInvert))))
